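# Physics asset for the upper and lower jacket flaps on the PlaidCarbon medium female cosmetic body.
# Each dynamic flap bone constrains to its own small base anchor with tight twist and asymmetric swing limits, much wider on the right side.
# Anchors and upper arms link to spine_05 through free angular constraints, with large spine and arm capsules forming body collision volumes.
# Right flaps use multiple thin capsules, whereas left flaps use one each.
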PhysicsAsset /BRCosmetics/Characters/Player/Female/Medium/Bodies/F_MED_PlaidCarbon/Meshes/F_MED_PlaidCarbon_JacketFlaps_Physics
bodies {
 base_jacktFlap_l [{shapeType=Capsule, length=0.5, radius=0.5, shapeName=base_jacktFlap_l_capsule, rotation=<rot 0 -90 0>}]
 base_jacktFlap_low_l [{shapeType=Capsule, length=0.5, radius=0.5, shapeName=base_jacktFlap_low_l_capsule, rotation=<rot 0 -90 0>}]
 base_jacktFlap_low_r [{shapeType=Capsule, length=0.5, radius=0.5, shapeName=base_jacktFlap_low_r_capsule, rotation=<rot 0 -90 0>}]
 base_jacktFlap_r [{shapeType=Capsule, length=0.5, radius=0.5, shapeName=base_jacktFlap_r_capsule, rotation=<rot 0 -90 0>}]
 dyn_jacktFlap_l [{
   shapeType=Capsule
   length=4.7305
   radius=2.2239
   shapeName=dyn_jacktFlap_l_capsule
   center=<4.7815 -0.5663 -1.1111>
   rotation=<rot 0 4.972 -85.6743>
  }]
 dyn_jacktFlap_low_l [{
   shapeType=Capsule
   length=4.0881
   radius=1.4191
   shapeName=dyn_jacktFlap_low_l_capsule
   center=<1.8564 -0.4165 -1.0515>
   rotation=<rot -0.0001 5.0004 89.9997>
  }]
 dyn_jacktFlap_low_r [
  {
   shapeType=Capsule
   length=9.6068
   radius=1
   shapeName=dyn_jacktFlap_low_r_capsule
   center=<1.4119 -1.1496 -0.5899>
   rotation=<rot -0.0004 4.9994 86.7835>
  }
  {
   shapeType=Capsule
   length=8.496
   radius=1
   shapeName=dyn_jacktFlap_low_r_capsule
   center=<3.7736 -2.9184 -0.5189>
   rotation=<rot -0.0001 4.9998 87.288>
  }
 ]
 dyn_jacktFlap_r [
  {
   shapeType=Capsule
   length=3.8321
   radius=1.1291
   shapeName=dyn_jacktFlap_r_capsule
   center=<6.7266 0.6223 -0.5684>
   rotation=<rot 0 -0.0001 -94.9999>
  }
  {
   shapeType=Capsule
   length=9.5918
   radius=0.9794
   shapeName=dyn_jacktFlap_r_capsule
   center=<4.1021 0.3848 -0.6367>
   rotation=<rot 0 -0.0001 -94.9999>
  }
  {
   shapeType=Capsule
   length=12.0892
   radius=0.8271
   shapeName=dyn_jacktFlap_r_capsule
   center=<1.5197 1.9799 -0.2041>
   rotation=<rot 0.3353 5.6576 -93.382>
  }
 ]
 spine_04 [{shapeType=Capsule, length=10.4914, radius=11.9883, shapeName=spine_04_capsule, center=<-1.4302 2.2281 -0.2417>}]
 spine_05 [{shapeType=Capsule, length=15.7751, radius=10.8029, shapeName=spine_05_capsule, center=<-0.9865 -1.2063 0>}]
 upperarm_l [{
   shapeType=Capsule
   length=15.9384
   radius=5.7879
   shapeName=upperarm_l_capsule
   center=<11.1418 -0.6726 -0.2909>
   rotation=<rot 90 -90.9147 -90.9151>
  }]
 upperarm_r [{
   shapeType=Capsule
   length=15.9384
   radius=5.7879
   shapeName=upperarm_l_capsule
   center=<-11.2858 0.9336 -0.2909>
   rotation=<rot 90 180 -177.4742>
  }]
}
constraints [
 {bone1Name=dyn_jacktFlap_l, bone2Name=base_jacktFlap_l, swing1LimitDegrees=10, swing2LimitDegrees=60, twistLimitDegrees=5}
 {bone1Name=dyn_jacktFlap_r, bone2Name=base_jacktFlap_r, swing1LimitDegrees=10, swing2LimitDegrees=140, twistLimitDegrees=10}
 {
  bone1Name=upperarm_l
  bone2Name=spine_05
  swing1LimitDegrees=45
  swing1Motion=Free
  swing2LimitDegrees=45
  swing2Motion=Free
  twistLimitDegrees=45
  twistMotion=Free
 }
 {
  bone1Name=upperarm_r
  bone2Name=spine_05
  swing1LimitDegrees=45
  swing1Motion=Free
  swing2LimitDegrees=45
  swing2Motion=Free
  twistLimitDegrees=45
  twistMotion=Free
 }
 {
  bone1Name=spine_05
  bone2Name=spine_04
  swing1LimitDegrees=45
  swing1Motion=Free
  swing2LimitDegrees=45
  swing2Motion=Free
  twistLimitDegrees=45
  twistMotion=Free
 }
 {bone1Name=dyn_jacktFlap_low_l, bone2Name=base_jacktFlap_low_l, swing1LimitDegrees=10, swing2LimitDegrees=60, twistLimitDegrees=5}
 {bone1Name=dyn_jacktFlap_low_r, bone2Name=base_jacktFlap_low_r, swing1LimitDegrees=10, swing2LimitDegrees=140, twistLimitDegrees=5}
 {
  bone1Name=base_jacktFlap_l
  bone2Name=spine_05
  swing1LimitDegrees=45
  swing1Motion=Free
  swing2LimitDegrees=45
  swing2Motion=Free
  twistLimitDegrees=45
  twistMotion=Free
 }
 {
  bone1Name=base_jacktFlap_low_l
  bone2Name=spine_05
  swing1LimitDegrees=45
  swing1Motion=Free
  swing2LimitDegrees=45
  swing2Motion=Free
  twistLimitDegrees=45
  twistMotion=Free
 }
 {
  bone1Name=base_jacktFlap_low_r
  bone2Name=spine_05
  swing1LimitDegrees=45
  swing1Motion=Free
  swing2LimitDegrees=45
  swing2Motion=Free
  twistLimitDegrees=45
  twistMotion=Free
 }
 {
  bone1Name=base_jacktFlap_r
  bone2Name=spine_05
  swing1LimitDegrees=45
  swing1Motion=Free
  swing2LimitDegrees=45
  swing2Motion=Free
  twistLimitDegrees=45
  twistMotion=Free
 }
]
