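# Physics asset for the neck collar of the female medium PlaidCarbon body cosmetic, setting up five collar flaps for physical motion.
# Each flap's larger dyn capsule is constrained to a small base capsule on its anchor bone, with a swing range on one axis only and zero limits on swing1 and twist.
# Unconstrained capsules on the head and two neck bones provide body shapes for the flaps to collide with.
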PhysicsAsset /BRCosmetics/Characters/Player/Female/Medium/Bodies/F_MED_PlaidCarbon/Meshes/F_MED_PlaidCarbon_NeckCollar_Physics
bodies {
 base_neckFlap_01_l [{shapeType=Capsule, length=0, radius=1, shapeName=base_neckFlap_01_l_capsule, rotation=<rot 0 -90 0>}]
 base_neckFlap_01_r [{shapeType=Capsule, length=0, radius=1, shapeName=base_neckFlap_01_r_capsule, rotation=<rot 0 -90 0>}]
 base_neckFlap_02_l [{shapeType=Capsule, length=0, radius=1, shapeName=base_neckFlap_02_l_capsule, rotation=<rot 0 -90 0>}]
 base_neckFlap_02_r [{shapeType=Capsule, length=0, radius=1, shapeName=base_neckFlap_02_r_capsule, rotation=<rot 0 -90 0>}]
 base_neckFlap_b [{shapeType=Capsule, length=0, radius=1, shapeName=base_neckFlap_b_capsule, rotation=<rot 0 -90 0>}]
 dyn_neckFlap_01_l [{
   shapeType=Capsule
   length=2.2041
   radius=2.9524
   shapeName=dyn_neckFlap_01_l_capsule
   center=<4.5331 0.3059 2.1863>
   rotation=<rot 0 -1.6516 -101.3169>
  }]
 dyn_neckFlap_01_r [{
   shapeType=Capsule
   length=2.2041
   radius=2.9524
   shapeName=dyn_neckFlap_01_l_capsule
   center=<5.1734 -0.1865 2.3146>
   rotation=<rot 0 178.3474 -84.1302>
  }]
 dyn_neckFlap_02_l [{
   shapeType=Capsule
   length=4.1198
   radius=3.0586
   shapeName=dyn_neckFlap_01_l_capsule
   center=<4.6347 0.3965 2.4562>
   rotation=<rot -0.0003 178.3488 -98.683>
  }]
 dyn_neckFlap_02_r [{
   shapeType=Capsule
   length=3.4145
   radius=3.0586
   shapeName=dyn_neckFlap_01_l_capsule
   center=<5.1756 -0.1082 2.269>
   rotation=<rot -0.0002 178.3473 -78.6831>
  }]
 dyn_neckFlap_b [{
   shapeType=Capsule
   length=6.7717
   radius=3.4591
   shapeName=dyn_neckFlap_01_l_capsule
   center=<4.7997 -0.1627 2.956>
   rotation=<rot -80.7357 90.5384 3.2301>
  }]
 head [
  {shapeType=Capsule, length=5.2171, radius=6.3495, shapeName=neck_01_capsule, center=<3.387 1.6564 0>, rotation=<rot 0 -90.0001 90>}
  {shapeType=Capsule, length=3.8171, radius=3.8661, shapeName=head_capsule, center=<-0.8579 5.877 0>, rotation=<rot -90 0 -0.0002>}
 ]
 neck_01 [{shapeType=Capsule, length=0.4022, radius=4.1078, shapeName=neck_01_capsule, center=<0 1.0449 0>, rotation=<rot 0 -90.0001 90>}]
 neck_02 [{shapeType=Capsule, length=0.4025, radius=4.1075, shapeName=neck_01_capsule, center=<0 0.492 0>, rotation=<rot 0 -90.0001 90>}]
}
constraints [
 {bone1Name=dyn_neckFlap_01_l, bone2Name=base_neckFlap_01_l, swing1LimitDegrees=0, swing2LimitDegrees=25, twistLimitDegrees=0}
 {bone1Name=dyn_neckFlap_01_r, bone2Name=base_neckFlap_01_r, swing1LimitDegrees=0, swing2LimitDegrees=25, twistLimitDegrees=0}
 {bone1Name=dyn_neckFlap_02_l, bone2Name=base_neckFlap_02_l, swing1LimitDegrees=0, swing2LimitDegrees=25, twistLimitDegrees=0}
 {bone1Name=dyn_neckFlap_02_r, bone2Name=base_neckFlap_02_r, swing1LimitDegrees=0, swing2LimitDegrees=25, twistLimitDegrees=0}
 {bone1Name=dyn_neckFlap_b, bone2Name=base_neckFlap_b, swing1LimitDegrees=0, swing2LimitDegrees=25, twistLimitDegrees=0}
]
