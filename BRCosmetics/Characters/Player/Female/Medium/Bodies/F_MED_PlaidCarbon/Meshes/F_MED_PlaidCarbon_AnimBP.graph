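# Post-process style cosmetic AnimBP for the F_MED_PlaidCarbon female body, adding secondary motion to its jacket, collar, grenades and zipper.
# Copies the parent mesh pose through a Control Rig, then chains cached-pose stages: additive cape-socket and jacket-flap offsets, AnimDynamics for grenades and zipper, and bool-bypassable RigidBody collar and flap simulation.
# EventGraph tracks weapon type, emotes, hang gliding, smoothed tactical-sprint jumps and state-based gravity overrides, including an inverse set.
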AnimBlueprint /BRCosmetics/Characters/Player/Female/Medium/Bodies/F_MED_PlaidCarbon/Meshes/F_MED_PlaidCarbon_AnimBP
target_skeleton=&/F_MED_PlaidCarbon_Skeleton
graphs [
 AnimGraph {nodes [
   Output Pose(Result=Use cached pose 'DynamicsCache'()) {layerGroup=None}
   n1 = "Transform (Modify) Bone - Bone: base_jacktFlap_l"!(
    ComponentPose=Local To Component(LocalPose=Use cached pose 'PreSimModCache'())
    bAlphaBoolEnabled:bool=True
    Translation="0,0,0"
    Rotation="-10,0,5"
   ) -> (Pose) {
    alphaInputType=Bool
    rotationMode=BMM_Additive
    rotationSpace=BCS_BoneSpace
    translationMode=BMM_Additive
    translationSpace=BCS_BoneSpace
    alphaBoolBlend {blendInTime=0.3, blendOption=Cubic, blendOutTime=0.3}
    boneToModify {boneName=base_jacktFlap_l}
    rotation {pitch=-10, roll=5}
   }
   n2 = "Transform (Modify) Bone - Bone: base_jacktFlap_l"(ComponentPose=@n1, bAlphaBoolEnabled=True, Translation="0,0,-1.5", Rotation="0,0,10") {
    alphaInputType=Bool
    rotationMode=BMM_Additive
    rotationSpace=BCS_BoneSpace
    translationMode=BMM_Additive
    translationSpace=BCS_BoneSpace
    alphaBoolBlend {blendInTime=0.3, blendOption=Cubic, blendOutTime=0.3}
    boneToModify {boneName=base_jacktFlap_l}
    rotation {roll=10}
    translation {z=-1.5}
   }
   Save cached pose 'CRCache_Input'(Pose=Blend Poses by bool(BlendPose_0=Component To Local(ComponentPose=@n2), BlendPose_1=Use cached pose 'PreSimModCache'()) {blendType=Linear}) {cacheName=CRCache_Input}
   n3 = RigidBody!(ComponentPose=Local To Component(LocalPose=Use cached pose 'AnimDynamicsCache'()), Alpha, OverrideWorldGravity:Vector="0,0,980") -> (Pose) {
    bOverrideWorldGravity=true
    overridePhysicsAsset=&/F_MED_PlaidCarbon_NeckCollar_Physics
    simulationSpace=BaseBoneSpace
    alphaBoolBlend {blendOption=Linear}
    alphaScaleBiasClamp {interpSpeedDecreasing=0.1, interpSpeedIncreasing=0.1}
    baseBoneRef {boneName=spine_05}
    componentAppliedLinearAccClamp=<3000 3000 3000>
    componentLinearAccScale=<0.01 0.01 0.01>
    componentLinearVelScale=<0.01 0.01 0.01>
    overrideWorldGravity {z=980}
    simSpaceSettings {maxAngularAcceleration=600, maxAngularVelocity=1000, maxLinearAcceleration=700, maxLinearVelocity=1000, externalLinearDragV=<0.2 0.2>}
   }
   n4 = RigidBody(ComponentPose=@n3, OverrideWorldGravity="0,0,0") {
    bOverrideWorldGravity=true
    overridePhysicsAsset=&/F_MED_PlaidCarbon_JacketFlaps_Physics
    simulationSpace=BaseBoneSpace
    alphaBoolBlend {blendOption=Linear}
    alphaScaleBiasClamp {interpSpeedDecreasing=0.1, interpSpeedIncreasing=0.1}
    baseBoneRef {boneName=spine_05}
    componentAppliedLinearAccClamp=<3000 3000 3000>
    componentLinearAccScale=<0.01 0.01 0.01>
    componentLinearVelScale=<0.01 0.01 0.01>
    overrideWorldGravity {z=980}
    simSpaceSettings {maxAngularAcceleration=1000, maxAngularVelocity=3, maxLinearAcceleration=700, maxLinearVelocity=500, externalLinearDragV {x=0.2}}
   }
   Save cached pose 'DynamicsCache'(Pose=Blend Poses by bool(BlendPose_0=Component To Local(ComponentPose=@n4), BlendPose_1=Use cached pose 'AnimDynamicsCache'()) {blendType=Linear}) {cacheName=DynamicsCache}
   n5 = Control Rig!(
    Source=Copy Pose From Mesh() {bUseAttachedParent=true}
    Alpha
    Holding Shotgun:bool=Get ShotGun_Out() -> (ShotGun_Out:bool)
    Holding Smg:bool=Get Smg_Out() -> (Smg_Out:bool)
    Holding AR:bool=Get AR_Out() -> (AR_Out:bool)
    areRigidBodiesEnabled:bool=Get bRigidBodyEnabled()
    isCrouching:bool=Get bIsCrouching()
    isTacticalJumping:double=Get IsTacticalSprintingFloat() -> (IsTacticalSprintingFloat:double)
   ) -> (Pose) {
    alphaBoolBlend {blendOption=Linear}
    controlRigAssetReference {blueprintRigClass=&/F_MED_PlaidCarbon_Body_CtrlRig.F_MED_PlaidCarbon_Body_CtrlRig_C}
   }
   Save cached pose 'PreSim_Cache'(Pose=@n5) {cacheName=PreSim_Cache}
   n6 = Anim Dynamics - Bone: dyn_granade_01_r!(ComponentPose=Local To Component(LocalPose=Use cached pose 'CRCache_Input'()), Alpha=1, GravityOverride:Vector="0,0,-980") -> (Pose) {
    angularDampingOverride=1
    angularSpringConstant=12000
    bAngularSpring=true
    bOverrideAngularDamping=true
    bUseGravityOverride=true
    gravityScale=0.35
    simulationSpace=RootRelative
    alphaBoolBlend {blendOption=Linear}
    boundBone {boneName=dyn_granade_01_r}
    componentAppliedLinearAccClamp=<3000 3000 3000>
    componentLinearAccScale=<0.2 0.2 0.2>
    componentLinearVelScale=<0.2 0.2 0.2>
    gravityOverride {z=-980}
    physicsBodyDefinitions [{
      boundBone {boneName=dyn_granade_01_r}
      boxExtents=<20 30 30>
      constraintSetup {angularLimitsMax=<0 3 10>, angularLimitsMin=<0 -15 -10>, angularTarget=<1 0 0>}
      localJointOffset=<2 0 0>
     }]
   }
   n7 = Anim Dynamics - Bone: dyn_granade_02_r!(ComponentPose=@n6, Alpha=1, GravityOverride:Vector="0,0,-980") -> (Pose) {
    angularDampingOverride=1
    angularSpringConstant=12000
    bAngularSpring=true
    bOverrideAngularDamping=true
    bUseGravityOverride=true
    gravityScale=0.35
    simulationSpace=RootRelative
    alphaBoolBlend {blendOption=Linear}
    boundBone {boneName=dyn_granade_02_r}
    componentAppliedLinearAccClamp=<3000 3000 3000>
    componentLinearAccScale=<0.2 0.2 0.2>
    componentLinearVelScale=<0.2 0.2 0.2>
    gravityOverride {z=-980}
    physicsBodyDefinitions [{
      boundBone {boneName=dyn_granade_02_r}
      boxExtents=<20 30 30>
      constraintSetup {angularLimitsMax=<0 3 10>, angularLimitsMin=<0 -15 -10>, angularTarget=<1 0 0>}
      localJointOffset=<2 0 0>
     }]
   }
   n8 = Anim Dynamics - Bone: dyn_zipper!(ComponentPose=@n7, Alpha=0.75, GravityOverride:Vector="0,0,-980") -> (Pose) {
    angularDampingOverride=1
    angularSpringConstant=20000
    bOverrideAngularDamping=true
    bUseGravityOverride=true
    gravityScale=0.35
    simulationSpace=RootRelative
    alphaBoolBlend {blendOption=Linear}
    boundBone {boneName=dyn_zipper}
    componentAppliedLinearAccClamp=<3000 3000 3000>
    componentLinearAccScale=<0.6 0.6 0.6>
    componentLinearVelScale=<0.6 0.6 0.6>
    gravityOverride {z=-980}
    physicsBodyDefinitions [{
      boundBone {boneName=dyn_zipper}
      boxExtents=<10 16 8>
      constraintSetup {angularLimitsMax=<5 25 45>, angularLimitsMin=<-5 -80 -45>, angularTarget=<0 0 0>}
      localJointOffset=<4 0 0>
     }]
   }
   Save cached pose 'AnimDynamicsCache'(Pose=Component To Local(ComponentPose=@n8)) {cacheName=AnimDynamicsCache}
   n9 = "Transform (Modify) Bone - Bone: attach_cape"!(
    ComponentPose=Local To Component(LocalPose=Use cached pose 'PreSim_Cache'())
    bAlphaBoolEnabled:bool=True
    Translation="0,-1,0"
    Rotation="0,0,16"
   ) -> (Pose) {
    alphaInputType=Bool
    rotationMode=BMM_Additive
    rotationSpace=BCS_BoneSpace
    translationMode=BMM_Additive
    translationSpace=BCS_BoneSpace
    alphaBoolBlend {blendInTime=0.5, blendOption=CircularOut, blendOutTime=0.5}
    boneToModify {boneName=attach_cape}
    rotation {roll=16}
    translation {y=-1}
   }
   n10 = "Transform (Modify) Bone - Bone: attach_cape"(ComponentPose=@n9, bAlphaBoolEnabled=True, Translation="0,-2,0", Rotation="0,0,20") {
    alphaInputType=Bool
    rotationMode=BMM_Additive
    rotationSpace=BCS_BoneSpace
    translationMode=BMM_Additive
    translationSpace=BCS_BoneSpace
    alphaBoolBlend {blendInTime=0.5, blendOption=CircularOut, blendOutTime=0.5}
    boneToModify {boneName=attach_cape}
    rotation {roll=20}
    translation {y=-2}
   }
   n11 = "Transform (Modify) Bone - Bone: attach_cape"(ComponentPose=@n10, bAlphaBoolEnabled=True, Translation="0,0,0", Rotation="0,0,10") {
    alphaInputType=Bool
    rotationMode=BMM_Additive
    rotationSpace=BCS_BoneSpace
    translationMode=BMM_Additive
    translationSpace=BCS_BoneSpace
    alphaBoolBlend {blendInTime=0.5, blendOption=Cubic, blendOutTime=0.5}
    boneToModify {boneName=attach_cape}
    rotation {roll=10}
   }
   Save cached pose 'PreSimModCache'(Pose=Component To Local(ComponentPose=@n11)) {cacheName=PreSimModCache}
  ]}
 EventGraph {nodes [
   Event OnBeginPlayingEmote() {is=title K2Node_Event}
   > Set isPlayingEmote(isPlayingEmote:bool=true) -> (Output_Get:bool)
   Event OnParachuteOpened()
   > Switch on EGliderType(Selection=Get PlayerGliderType() -> (PlayerGliderType=HangGlider))
   > Set Is Using Hang Gliding in Parachute(Is Using Hang Gliding in Parachute:bool=true) -> (Output_Get:bool)
   Event OnParachuteClosed()
   > Set Is Using Hang Gliding in Parachute(Is Using Hang Gliding in Parachute=false)
   n1 = Event BlueprintUpdateAnimation()
   > n2 = Sequence!(execute) -> (then_0, then_1, then_2:Exec, then_3:Exec)
   > n3 = Gravity Override State Function DEPRECATED(CharacterPartAnimInstance=Self-Reference(), GameTimeInSeconds=GetGameTimeInSeconds(), GravityOverrideParams=Get Gravity Override Params())
   > Set GravityOverride(GravityOverride=@n3)
   > n4 = Gravity Override State Function DEPRECATED(
    CharacterPartAnimInstance=Self-Reference()
    GameTimeInSeconds=GetGameTimeInSeconds()
    GravityOverrideParams=Get Gravity Override Params Inverse() -> (Gravity Override Params Inverse:Gravity Override Params Struct Structure)
   )
   > "Set GravityOverride  Inverse"("GravityOverride  Inverse":Vector=@n4) -> (Output_Get:Vector)
   Branch(execute=@n2.then_3, Condition=Get bIsPlayingEmote())
   n5 = GetAnimInstance(self=GetOwnerSkeletalMeshComponent())
   > Is Valid(InputObject=@n5)
   > Set isPlayingJubilation(isPlayingJubilation:bool=OR Boolean(
     A=Montage_IsPlaying(self=@n5, Montage=/Game/Animation/Game/MainPlayer/Emotes/Emote_Celebration_M)
     B=Montage_IsPlaying(self=@n5, Montage=/BRCosmetics/Animation/Game/MainPlayer/Emotes/Emote_Celebration_Female_M)
    )) -> (Output_Get:bool)
   n6 = Cast To FortPlayerAnimInstance(execute=@n2.then_1, Object=GetAnimInstance(self=GetOwnerSkeletalMeshComponent()))
   > Set Smg_Out(Smg_Out:bool=false) -> (Output_Get:bool)
   > Set ShotGun_Out(ShotGun_Out:bool=false) -> (Output_Get:bool)
   > Set AR_Out(AR_Out:bool=false) -> (Output_Get:bool)
   > n7 = Switch on EFortWeaponCoreAnimation!(execute, Selection:EFortWeaponCoreAnimation Enum=Get WeaponCoreAnim(self=@n6)) -> (
    Melee:Exec
    Pistol:Exec
    Shotgun:Exec
    PaperBlueprint:Exec
    Rifle:Exec
    MeleeOneHand:Exec
    MachinePistol:Exec
    RocketLauncher:Exec
    GrenadeLauncher:Exec
    GoingCommando:Exec
    AssaultRifle:Exec
    TacticalShotgun:Exec
    SniperRifle:Exec
    TrapPlacement:Exec
    ShoulderLauncher:Exec
    AbilityDecoTool:Exec
    Crossbow:Exec
    C4:Exec
    RemoteControl:Exec
    DualWield:Exec
    AR_BullPup:Exec
    AR_ForwardGrip:Exec
    MedPackPaddles:Exec
    SMG_P90:Exec
    AR_DrumGun:Exec
    Consumable_Small:Exec
    Consumable_Large:Exec
    Balloon:Exec
    MountedTurret:Exec
    CreativeTool:Exec
    ExplosiveBow:Exec
    AshtonIndigo:Exec
    AshtonChicago:Exec
    MeleeDualWield:Exec
    Unarmed:Exec
   )
   Branch(execute=@n7.AssaultRifle)
   > Set AR_Out(AR_Out=true)
   Branch(execute=@n7.SMG_P90)
   > Set Smg_Out(Smg_Out=true)
   Branch(execute=[@n7.Shotgun, @n7.TacticalShotgun])
   > Set ShotGun_Out(ShotGun_Out=true)
   n8 = Get IsTacticalJumping() -> (IsTacticalJumping:bool)
   Set IsTacticalSprintingFloat(
    execute=@n2.then_2
    IsTacticalSprintingFloat:double=FInterpTo(Current=Get IsTacticalSprintingFloat(), Target="To Float (Boolean)"(InBool=@n8), DeltaTime=@n1.DeltaTimeX, InterpSpeed=5)
   ) -> (Output_Get:double)
   Event OnEndPlayingEmote() {is=title K2Node_Event}
   > Set isPlayingEmote(isPlayingEmote=false)
   > Set isPlayingJubilation(isPlayingJubilation=false)
   Event OnBeginJumping()
   > Set IsTacticalJumping(IsTacticalJumping:bool=Get bIsTacticalSprinting()) -> (Output_Get:bool)
   Event OnEndJumping()
   > Set IsTacticalJumping(IsTacticalJumping=false)
   Set isPlayingZombie(isPlayingZombie:bool=false) -> (Output_Get:bool)
   Set isPlayingZombie(isPlayingZombie=Montage_IsPlaying(self=@n5, Montage=/BRCosmetics/Animation/Game/MainPlayer/Emotes/ZombieWalk/Emote_ZombieWalk_CMF_M))
  ]}
]
variables [
 AR_Out {type=boolean}
 Gravity Override Params {type=GravityOverrideParamsStruct}
 Gravity Override Params Inverse {
  type=GravityOverrideParamsStruct
  default {
   dBNO {gravityOverride=<980 400 0>}
   emoteOrMelee {gravityOverride=<0 0 980>}
   front_End {gravityOverride=<0 0 980>}
   onGround_Falling {gravityOverride=<0 0 980>}
   onGround_Moving {gravityOverride=<0 0 980>}
   onGround_Standing {gravityOverride=<0 0 980>}
   rail_Grinding {gravityOverride=<0 0 980>}
   riding_Moving {gravityOverride=<0 0 980>}
   riding_Moving_Targeting {gravityOverride=<0 0 980>}
   riding_Standing {gravityOverride=<0 0 980>}
   skydive_DivingDown {jointName=root, gravityOverride=<0 0 10>, windAmplitude=<100 0 500>, windFrequency=<20 0 20>}
   skydive_FloatingDown {gravityOverride=<0 0 980>, windAmplitude=<100 0 500>, windFrequency=<20 0 20>, windOffset=<0 0 0>}
   skydive_GoingUp {gravityOverride=<0 0 980>, windAmplitude=<0 0 0>, windFrequency=<0 0 0>}
   skydive_Parachuting {gravityOverride=<0 -800 980>, windAmplitude=<10 -250 200>, windFrequency=<1.5 3 8>}
   skydive_Surfing {gravityOverride=<0 0 980>, windAmplitude=<0 0 0>, windFrequency=<0 0 0>}
   swimming {jointName=root, gravityOverride=<0 0 980>}
   tactical_Sprint {gravityOverride=<0 -250 600>}
   ziplining {gravityOverride=<0 0 980>}
  }
 }
 GravityOverride {type=Vector}
 "GravityOverride  Inverse" {type=Vector}
 Is Using Hang Gliding in Parachute {type=boolean}
 isPlayingEmote {type=boolean}
 isPlayingJubilation {type=boolean}
 isPlayingZombie {type=boolean}
 IsTacticalJumping {type=boolean}
 IsTacticalSprintingFloat {type=number}
 ShotGun_Out {type=boolean}
 Smg_Out {type=boolean}
]
