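# Physics asset for the F_MED_PlaidCarbon female face accessory mesh, covering ponytail, hair strands and neck and jacket flaps for secondary motion.
# Small base bodies link the six-segment ponytail and four side strands to the head, with limited swing and little twist down the chains, except the free ponytail root.
# Flaps attach to spine_05 by free constraints.
# Head, neck, upper spine and right arm shapes provide body volumes, with no left arm.
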PhysicsAsset /BRCosmetics/Characters/Player/Female/Medium/Bodies/F_MED_PlaidCarbon/Meshes/Parts/F_MED_PlaidCarbon_FaceAcc_Physics
bodies {
 base_leftHair_A [{shapeType=Capsule, length=0.505, radius=0.505, shapeName=base_leftHair_A_capsule, rotation=<rot 0 -90 0>}]
 base_leftHair_B [{shapeType=Capsule, length=0.505, radius=0.505, shapeName=base_leftHair_B_capsule, rotation=<rot 0 -90 0>}]
 base_ponytail [{shapeType=Capsule, length=0.505, radius=0.505, shapeName=base_ponytail_capsule, rotation=<rot 0 -90 0>}]
 base_rightHair_A [{shapeType=Capsule, length=0.505, radius=0.505, shapeName=base_rightHair_A_capsule, rotation=<rot 0 -90 0>}]
 base_rightHair_B [{shapeType=Capsule, length=0.505, radius=0.505, shapeName=base_rightHair_B_capsule, rotation=<rot 0 -90 0>}]
 clavicle_r [
  {shapeType=Sphere, radius=5.2537, shapeName=clavicle_r_sphere, center=<-10.4735 2.4899 2.2302>}
  {shapeType=Capsule, length=7.5086, radius=2.4964, shapeName=clavicle_r_capsule, center=<-5.4252 1.4955 0.7565>, rotation=<rot -0.0001 -90 90>}
  {shapeType=Capsule, length=5.4453, radius=4.3768, shapeName=clavicle_r_capsule, center=<-5.6217 -3.3217 5.2487>, rotation=<rot 0 0 -29.3677>}
 ]
 dyn_jacktFlap_r [
  {
   shapeType=Capsule
   length=6.184
   radius=1.1302
   shapeName=dyn_jacktFlap_r_capsule
   center=<3.6781 0.8 -0.3662>
   rotation=<rot 0.0001 -90 -94.6043>
  }
  {
   shapeType=Capsule
   length=3.8369
   radius=0.8749
   shapeName=dyn_jacktFlap_r_capsule
   center=<2.5503 -2.5768 -0.3601>
   rotation=<rot 82.2199 -160.9781 158.2481>
  }
 ]
 dyn_leftHair_A [{shapeType=Capsule, length=4.7914, radius=2, shapeName=dyn_leftHair_A_capsule, center=<2 0 0>, rotation=<rot 90 0 0>}]
 dyn_leftHair_B [{shapeType=Capsule, length=3.6252, radius=2, shapeName=dyn_leftHair_B_capsule, center=<2.0577 0 0.2443>, rotation=<rot 76.7117 0 0>}]
 dyn_neckFlap_01_r [{
   shapeType=Capsule
   length=3
   radius=1.7576
   shapeName=dyn_neckFlap_01_r_capsule
   center=<1.9638 0 -0.5148>
   rotation=<rot 85.3265 140.4503 119.191>
  }]
 dyn_neckFlap_02_r [{
   shapeType=Capsule
   length=3
   radius=1.2395
   shapeName=dyn_neckFlap_02_r_capsule
   center=<1.9638 0 -0.5148>
   rotation=<rot 77.3582 166.5804 166.2597>
  }]
 dyn_neckFlap_b [{
   shapeType=Capsule
   length=3
   radius=1.2395
   shapeName=dyn_neckFlap_b_capsule
   center=<2.1158 -0.0363 -1.2118>
   rotation=<rot 77.358 166.5808 166.2599>
  }]
 dyn_ponytail_01 [{shapeType=Capsule, length=2, radius=2.5361, shapeName=dyn_ponytail_01_capsule, center=<2 -1.1999 0>, rotation=<rot 90 0 0>}]
 dyn_ponytail_02 [{shapeType=Capsule, length=3, radius=2, shapeName=dyn_ponytail_02_capsule, center=<2 -1.1834 0>, rotation=<rot 90 0 0>}]
 dyn_ponytail_03 [{
   shapeType=Capsule
   length=4.2149
   radius=2.4884
   shapeName=dyn_ponytail_03_capsule
   center=<2 0 -0.7348>
   rotation=<rot 85.3835 0.0002 0.0001>
  }]
 dyn_ponytail_04 [
  {
   shapeType=Capsule
   length=4.7357
   radius=1.9786
   shapeName=dyn_ponytail_04_capsule
   center=<4.4199 -0.1399 -1.1266>
   rotation=<rot 85.2634 0.001 0.1182>
  }
  {
   shapeType=Capsule
   length=3.8251
   radius=2.0705
   shapeName=dyn_ponytail_04_capsule
   center=<4.8991 1.8269 0.7917>
   rotation=<rot -35.5618 -83.154 86.0061>
  }
 ]
 dyn_ponytail_05 [{shapeType=Capsule, length=2.1419, radius=2.8581, shapeName=dyn_ponytail_05_capsule, center=<2 0 0>, rotation=<rot 90 0 0>}]
 dyn_ponytail_06 [{shapeType=Capsule, length=3, radius=2, shapeName=dyn_ponytail_06_capsule, center=<2 0 0>, rotation=<rot 90 0 0>}]
 dyn_rightHair_A [{shapeType=Capsule, length=3.2964, radius=2, shapeName=dyn_rightHair_A_capsule, center=<2 0 0>, rotation=<rot 90 0 0>}]
 dyn_rightHair_B [{
   shapeType=Capsule
   length=6.9444
   radius=2
   shapeName=dyn_rightHair_B_capsule
   center=<2 0 0.4562>
   rotation=<rot 82.8857 -179.9999 -179.9999>
  }]
 head [{
   shapeType=Capsule
   length=9.7113
   radius=6.0834
   shapeName=head_capsule
   center=<-2.2245 1.1921 0.4421>
   rotation=<rot -21.9911 -93.9657 -88.4524>
  }]
 lowerarm_r [{
   shapeType=Capsule
   length=14.7634
   radius=4.2271
   shapeName=lowerarm_r_capsule
   center=<-0.7657 -1.1341 0.2189>
   rotation=<rot 5.6188 -114.6143 89.4749>
  }]
 neck_01 [{shapeType=Capsule, length=5.1194, radius=4.851, shapeName=neck_01_capsule, center=<1.1565 6.6354 0.7902>, rotation=<rot 0 -90 0>}]
 neck_02 [{shapeType=Capsule, length=0.505, radius=6.505, shapeName=neck_02_capsule, rotation=<rot 0 -90 0>}]
 spine_05 [
  {shapeType=Capsule, length=15.0785, radius=4.6832, shapeName=spine_05_capsule, center=<4.5592 -2.4128 5.5943>, rotation=<rot 0 -90 0>}
  {
   shapeType=Capsule
   length=9.6814
   radius=9.6585
   shapeName=spine_05_capsule
   center=<12.8111 -16.6908 7.3369>
   rotation=<rot 0 -0.0001 4.3241>
  }
 ]
 upperarm_r [{
   shapeType=Capsule
   length=21.0804
   radius=4.6629
   shapeName=upperarm_r_capsule
   center=<-15.579 -0.3111 1.229>
   rotation=<rot -0.025 -87.8618 87.0168>
  }]
}
constraints [
 {
  bone1Name=head
  bone2Name=neck_02
  swing1LimitDegrees=45
  swing1Motion=Free
  swing2LimitDegrees=45
  swing2Motion=Free
  twistLimitDegrees=45
  twistMotion=Free
 }
 {
  bone1Name=upperarm_r
  bone2Name=clavicle_r
  swing1LimitDegrees=45
  swing1Motion=Free
  swing2LimitDegrees=45
  swing2Motion=Free
  twistLimitDegrees=45
  twistMotion=Free
 }
 {
  bone1Name=dyn_ponytail_01
  bone2Name=base_ponytail
  swing1LimitDegrees=45
  swing1Motion=Free
  swing2LimitDegrees=45
  swing2Motion=Free
  twistLimitDegrees=45
  twistMotion=Free
 }
 {bone1Name=dyn_ponytail_02, bone2Name=dyn_ponytail_01, swing1LimitDegrees=10, swing2LimitDegrees=10, twistLimitDegrees=0}
 {bone1Name=dyn_ponytail_03, bone2Name=dyn_ponytail_02, swing1LimitDegrees=25, swing2LimitDegrees=25, twistLimitDegrees=0}
 {bone1Name=dyn_ponytail_04, bone2Name=dyn_ponytail_03, swing1LimitDegrees=20, swing2LimitDegrees=20, twistLimitDegrees=0}
 {bone1Name=dyn_ponytail_05, bone2Name=dyn_ponytail_04, swing1LimitDegrees=60, swing2LimitDegrees=60, twistLimitDegrees=1}
 {bone1Name=dyn_ponytail_06, bone2Name=dyn_ponytail_05, swing1LimitDegrees=45, swing2LimitDegrees=45, twistLimitDegrees=0}
 {
  bone1Name=base_ponytail
  bone2Name=head
  swing1LimitDegrees=45
  swing1Motion=Free
  swing2LimitDegrees=45
  swing2Motion=Free
  twistLimitDegrees=45
  twistMotion=Free
 }
 {
  bone1Name=base_leftHair_A
  bone2Name=head
  swing1LimitDegrees=45
  swing1Motion=Free
  swing2LimitDegrees=45
  swing2Motion=Free
  twistLimitDegrees=45
  twistMotion=Free
 }
 {
  bone1Name=base_leftHair_B
  bone2Name=head
  swing1LimitDegrees=45
  swing1Motion=Free
  swing2LimitDegrees=45
  swing2Motion=Free
  twistLimitDegrees=45
  twistMotion=Free
 }
 {
  bone1Name=base_rightHair_A
  bone2Name=head
  swing1LimitDegrees=45
  swing1Motion=Free
  swing2LimitDegrees=45
  swing2Motion=Free
  twistLimitDegrees=45
  twistMotion=Free
 }
 {
  bone1Name=base_rightHair_B
  bone2Name=head
  swing1LimitDegrees=45
  swing1Motion=Free
  swing2LimitDegrees=45
  swing2Motion=Free
  twistLimitDegrees=45
  twistMotion=Free
 }
 {bone1Name=dyn_rightHair_B, bone2Name=base_rightHair_B, swing1LimitDegrees=3.5, swing2LimitDegrees=6, twistLimitDegrees=0.5}
 {bone1Name=dyn_rightHair_A, bone2Name=base_rightHair_A, swing1LimitDegrees=3.5, swing2LimitDegrees=6, twistLimitDegrees=0.5}
 {bone1Name=dyn_leftHair_B, bone2Name=base_leftHair_B, swing1LimitDegrees=3.5, swing2LimitDegrees=6, twistLimitDegrees=0.5}
 {bone1Name=dyn_leftHair_A, bone2Name=base_leftHair_A, swing1LimitDegrees=3.5, swing2LimitDegrees=6, twistLimitDegrees=0.5}
 {
  bone1Name=neck_02
  bone2Name=neck_01
  swing1LimitDegrees=45
  swing1Motion=Free
  swing2LimitDegrees=45
  swing2Motion=Free
  twistLimitDegrees=45
  twistMotion=Free
 }
 {
  bone1Name=clavicle_r
  bone2Name=spine_05
  swing1LimitDegrees=45
  swing1Motion=Free
  swing2LimitDegrees=45
  swing2Motion=Free
  twistLimitDegrees=45
  twistMotion=Free
 }
 {
  bone1Name=dyn_neckFlap_02_r
  bone2Name=spine_05
  swing1LimitDegrees=45
  swing1Motion=Free
  swing2LimitDegrees=45
  swing2Motion=Free
  twistLimitDegrees=45
  twistMotion=Free
 }
 {
  bone1Name=dyn_neckFlap_01_r
  bone2Name=spine_05
  swing1LimitDegrees=45
  swing1Motion=Free
  swing2LimitDegrees=45
  swing2Motion=Free
  twistLimitDegrees=45
  twistMotion=Free
 }
 {
  bone1Name=dyn_neckFlap_b
  bone2Name=spine_05
  swing1LimitDegrees=45
  swing1Motion=Free
  swing2LimitDegrees=45
  swing2Motion=Free
  twistLimitDegrees=45
  twistMotion=Free
 }
 {
  bone1Name=dyn_jacktFlap_r
  bone2Name=spine_05
  swing1LimitDegrees=45
  swing1Motion=Free
  swing2LimitDegrees=45
  swing2Motion=Free
  twistLimitDegrees=45
  twistMotion=Free
 }
 {
  bone1Name=neck_01
  bone2Name=spine_05
  swing1LimitDegrees=45
  swing1Motion=Free
  swing2LimitDegrees=45
  swing2Motion=Free
  twistLimitDegrees=45
  twistMotion=Free
 }
 {
  bone1Name=lowerarm_r
  bone2Name=upperarm_r
  swing1LimitDegrees=45
  swing1Motion=Free
  swing2LimitDegrees=45
  swing2Motion=Free
  twistLimitDegrees=45
  twistMotion=Free
 }
]
