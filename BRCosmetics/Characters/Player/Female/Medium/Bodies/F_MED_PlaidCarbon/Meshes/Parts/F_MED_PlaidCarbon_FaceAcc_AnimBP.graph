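# Post-process style Animation Blueprint for the F_MED_PlaidCarbon face accessory, animating hat, hair and ponytail against the body.
# Parent and Body_Mesh poses blend at the clavicles, then feed a Control Rig driven by crouch, backpack, weapon, targeting and zombie-emote state.
# Cached stages chain weighted ponytail constraints, right/back neck and jacket flap transforms sampled from body sockets, RigidBody and AnimDynamics, then additive ponytail offsets.
# Events also track gliding and gravity override.
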AnimBlueprint /BRCosmetics/Characters/Player/Female/Medium/Bodies/F_MED_PlaidCarbon/Meshes/Parts/F_MED_PlaidCarbon_FaceAcc_AnimBP
target_skeleton=&/F_MED_PlaidCarbon_FaceAcc_Skeleton
graphs [
 AnimGraph {nodes [
   Output Pose(Result=Use cached pose 'CRCache_offset'()) {layerGroup=None}
   n1 = Get Body_Mesh()
   n2 = Control Rig!(
    Source=Layered blend per bone(BasePose=Copy Pose From Mesh() {bUseAttachedParent=true}, BlendPoses_0=Copy Pose From Mesh(SourceMeshComponent=@n1)) {layerSetup [{branchFilters [{boneName=clavicle_r}, {boneName=clavicle_l}]}]}
    Alpha
    Is crouching:bool=Get fCrouching() -> (fCrouching:bool)
    Is backpack On:bool=Get isBackpackEquipped() -> (isBackpackEquipped:bool)
    isTargeting:double=Get isTargeting() -> (isTargeting:double)
    isGunOut:double=Get isGunOut() -> (isGunOut:double)
    isPlayingZombieEmote:bool=Get Is Custom Emote Playing() -> (Is Custom Emote Playing:bool)
   ) -> (Pose) {
    alphaBoolBlend {blendOption=Linear}
    controlRigAssetReference {blueprintRigClass=&/F_MED_PlaidCarbon_FaceAcc_CtrlRig.F_MED_PlaidCarbon_FaceAcc_CtrlRig_C}
   }
   Save cached pose 'CRCache_Input'(Pose=@n2) {cacheName=CRCache_Input}
   n3 = "Constraint - dyn_ponytail_01 (PPPPP)"!(
    ComponentPose=Local To Component(LocalPose=Use cached pose 'CRCache_Input'())
    Alpha=0.3
    ConstraintWeights_0=0.4804
    ConstraintWeights_1:float=0.2974
    ConstraintWeights_2:float=0.1978
    ConstraintWeights_3:float=0.0142
    ConstraintWeights_4:float=0.0102
   ) -> (Pose) {
    alphaBoolBlend {blendOption=Linear}
    boneToModify {boneName=dyn_ponytail_01}
    constraintSetup [
     {targetBone {boneName=dyn_neckFlap_b}}
     {targetBone {boneName=dyn_neckFlap_02_r}}
     {targetBone {boneName=dyn_neckFlap_02_l}}
     {targetBone {boneName=dyn_neckFlap_01_l}}
     {targetBone {boneName=dyn_neckFlap_01_r}}
    ]
    constraintWeights [1,1,1,1,1]
   }
   n4 = "Constraint - dyn_ponytail_02 (PPPP)"!(
    ComponentPose=@n3
    Alpha=0.3
    ConstraintWeights_0=0.6158
    ConstraintWeights_1:float=0.2847
    ConstraintWeights_2:float=0.058
    ConstraintWeights_3:float=0.0416
   ) -> (Pose) {
    alphaBoolBlend {blendOption=Linear}
    boneToModify {boneName=dyn_ponytail_02}
    constraintSetup [
     {targetBone {boneName=dyn_neckFlap_02_r}}
     {targetBone {boneName=dyn_neckFlap_b}}
     {targetBone {boneName=dyn_neckFlap_01_r}}
     {targetBone {boneName=dyn_neckFlap_02_l}}
    ]
    constraintWeights [1,1,1,1]
   }
   n5 = "Constraint - dyn_ponytail_02 (PPPP)"(ComponentPose=@n4, Alpha=0.3, ConstraintWeights_0=0.6158, ConstraintWeights_1=0.2847, ConstraintWeights_2=0.058, ConstraintWeights_3=0.0416) {
    alphaBoolBlend {blendOption=Linear}
    boneToModify {boneName=dyn_ponytail_02}
    constraintSetup [
     {targetBone {boneName=dyn_neckFlap_02_r}}
     {targetBone {boneName=dyn_neckFlap_b}}
     {targetBone {boneName=dyn_neckFlap_01_r}}
     {targetBone {boneName=dyn_neckFlap_02_l}}
    ]
    constraintWeights [1,1,1,1]
   }
   n6 = "Constraint - dyn_ponytail_04 (PPPPPPPT)"!(
    ComponentPose=@n5
    Alpha=0.2
    ConstraintWeights_0=0.7991
    ConstraintWeights_1:float=0.0816
    ConstraintWeights_2:float=0.0611
    ConstraintWeights_3:float=0.0181
    ConstraintWeights_4:float=0.0142
    ConstraintWeights_5:float=0.0134
    ConstraintWeights_6:float=0.0125
    ConstraintWeights_7:float=0
   ) -> (Pose) {
    alphaBoolBlend {blendOption=Linear}
    boneToModify {boneName=dyn_ponytail_04}
    constraintSetup [
     {targetBone {boneName=clavicle_r}}
     {targetBone {boneName=clavicle_counterSkin_r}}
     {targetBone {boneName=spine_05}}
     {targetBone {boneName=upperarm_twist_01_r}}
     {targetBone {boneName=spine_04}}
     {targetBone {boneName=deform_upperarm_fr_r}}
     {targetBone {boneName=deform_upperarm_up_r}}
     {transformType=Translation, targetBone {boneName=None}}
    ]
    constraintWeights [1,1,1,1,1,1,1,1]
   }
   Save cached pose 'CRCache_constraints'(Pose=Component To Local(ComponentPose=@n6)) {cacheName=CRCache_constraints}
   n7 = RigidBody!(
    ComponentPose=Local To Component(LocalPose=Use cached pose 'CRCache_JacketFlaps'())
    Alpha=0
    OverrideWorldGravity:Vector="0,0,0"
    ExternalForce
   ) -> (Pose) {
    bOverrideWorldGravity=true
    overridePhysicsAsset=&/F_MED_PlaidCarbon_FaceAcc_Physics
    simulationSpace=BaseBoneSpace
    alphaBoolBlend {blendOption=Linear}
    alphaScaleBiasClamp {bInterpResult=true}
    baseBoneRef {boneName=spine_03}
    componentAppliedLinearAccClamp=<5000 5000 5000>
    simSpaceSettings {
     maxAngularAcceleration=500
     maxAngularVelocity=5
     maxLinearAcceleration=1000
     maxLinearVelocity=700
     velocityScaleZ=0.5
     worldAlpha=1
     externalLinearDragV=<0.2 0.2 0.2>
    }
   }
   n8 = Anim Dynamics - Bone: base_hair_C!(ComponentPose=@n7, Alpha=1, GravityOverride:Vector="0,0,-980", ExternalForce="0,0,0") -> (Pose) {
    angularDampingOverride=1
    angularSpringConstant=20000
    bAngularSpring=true
    bOverrideAngularDamping=true
    bUseGravityOverride=true
    gravityScale=0.35
    simulationSpace=RootRelative
    alphaBoolBlend {blendOption=Linear}
    boundBone {boneName=base_hair_C}
    componentLinearAccScale=<1 1 1>
    componentLinearVelScale=<1 1 1>
    gravityOverride {z=-980}
    physicsBodyDefinitions [{
      boundBone {boneName=base_hair_C}
      boxExtents=<20 30 30>
      constraintSetup {angularLimitsMax=<15 15 15>, angularLimitsMin=<-15 -15 -15>, angularTarget=<1 0 0>}
      localJointOffset=<4 0 0>
     }]
   }
   n9 = Anim Dynamics - Bone: dyn_hat!(ComponentPose=@n8, Alpha=1, GravityOverride:Vector="0,0,-980", ExternalForce="0,0,0") -> (Pose) {
    angularDampingOverride=1
    angularSpringConstant=15000
    bAngularSpring=true
    bOverrideAngularDamping=true
    bUseGravityOverride=true
    gravityScale=0.35
    simulationSpace=BoneRelative
    alphaBoolBlend {blendOption=Linear}
    boundBone {boneName=dyn_hat}
    componentLinearAccScale=<1 1 1>
    componentLinearVelScale=<1 1 1>
    gravityOverride {z=-980}
    physicsBodyDefinitions [{
      boundBone {boneName=dyn_hat}
      boxExtents=<18 28 28>
      constraintSetup {angularLimitsMax=<3 3 2>, angularLimitsMin=<-3 -3 -2>, angularTarget=<1 0 0>}
      localJointOffset=<4 0 0>
     }]
    relativeSpaceBone {boneName=head}
   }
   n10 = Anim Dynamics - Bone: dyn_ponytail_03!(ComponentPose=@n9, bAlphaBoolEnabled:bool=True, GravityOverride:Vector="0,0,-980") -> (Pose) {
    alphaInputType=Bool
    angularDampingOverride=1
    angularSpringConstant=10000
    bAngularSpring=true
    bOverrideAngularDamping=true
    bUseGravityOverride=true
    gravityScale=0.35
    simulationSpace=BoneRelative
    alphaBoolBlend {blendInTime=0.1, blendOption=Linear, blendOutTime=0.1}
    boundBone {boneName=dyn_ponytail_03}
    componentAppliedLinearAccClamp=<10000 10000 10000>
    componentLinearAccScale=<1 1 1>
    componentLinearVelScale=<1 1 1>
    gravityOverride {z=-980}
    physicsBodyDefinitions [{
      boundBone {boneName=dyn_ponytail_03}
      boxExtents=<20 20 20>
      constraintSetup {angularLimitsMax=<0 10 10>, angularLimitsMin=<0 -10 -10>, angularTarget=<1 0 0>}
      localJointOffset=<2 0 0>
     }]
    relativeSpaceBone {boneName=spine_03}
   }
   n11 = Anim Dynamics - Bone: dyn_ponytail_04!(ComponentPose=@n10, bAlphaBoolEnabled:bool=True, GravityOverride:Vector="0,0,-980") -> (Pose) {
    alphaInputType=Bool
    angularDampingOverride=1
    angularSpringConstant=10000
    bAngularSpring=true
    bOverrideAngularDamping=true
    bUseGravityOverride=true
    gravityScale=0.35
    simulationSpace=BoneRelative
    alphaBoolBlend {blendInTime=0.1, blendOption=Linear, blendOutTime=0.1}
    boundBone {boneName=dyn_ponytail_04}
    componentAppliedLinearAccClamp=<10000 10000 10000>
    componentLinearAccScale=<1 1 1>
    componentLinearVelScale=<1 1 1>
    gravityOverride {z=-980}
    physicsBodyDefinitions [{
      boundBone {boneName=dyn_ponytail_04}
      boxExtents=<20 20 20>
      constraintSetup {angularLimitsMax=<0 5 5>, angularLimitsMin=<0 -5 -5>, angularTarget=<1 0 0>}
      localJointOffset=<2 0 0>
     }]
    relativeSpaceBone {boneName=spine_03}
   }
   n12 = Anim Dynamics - Bone: dyn_ponytail_05!(ComponentPose=@n11, bAlphaBoolEnabled:bool=True, GravityOverride:Vector="0,0,-980") -> (Pose) {
    alphaInputType=Bool
    angularDampingOverride=1
    angularSpringConstant=10000
    bAngularSpring=true
    bOverrideAngularDamping=true
    bUseGravityOverride=true
    gravityScale=0.35
    simulationSpace=BoneRelative
    alphaBoolBlend {blendInTime=0.1, blendOption=Linear, blendOutTime=0.1}
    boundBone {boneName=dyn_ponytail_05}
    componentAppliedLinearAccClamp=<10000 10000 10000>
    componentLinearAccScale=<1 1 1>
    componentLinearVelScale=<1 1 1>
    gravityOverride {z=-980}
    physicsBodyDefinitions [{
      boundBone {boneName=dyn_ponytail_05}
      boxExtents=<25 25 25>
      constraintSetup {angularLimitsMax=<0 2.5 2.5>, angularLimitsMin=<0 -2.5 -2.5>, angularTarget=<1 0 0>}
      localJointOffset=<4 0 0>
     }]
    relativeSpaceBone {boneName=spine_03}
   }
   n13 = Anim Dynamics - Bone: dyn_ponytail_06!(ComponentPose=@n12, bAlphaBoolEnabled:bool=True, GravityOverride:Vector="0,0,-980") -> (Pose) {
    alphaInputType=Bool
    angularDampingOverride=1
    angularSpringConstant=10000
    bAngularSpring=true
    bOverrideAngularDamping=true
    bUseGravityOverride=true
    gravityScale=0.35
    simulationSpace=BoneRelative
    alphaBoolBlend {blendInTime=0.1, blendOption=Linear, blendOutTime=0.1}
    boundBone {boneName=dyn_ponytail_06}
    componentAppliedLinearAccClamp=<10000 10000 10000>
    componentLinearAccScale=<1 1 1>
    componentLinearVelScale=<1 1 1>
    gravityOverride {z=-980}
    physicsBodyDefinitions [{
      boundBone {boneName=dyn_ponytail_06}
      boxExtents=<25 25 25>
      constraintSetup {angularLimitsMax=<0 5 5>, angularLimitsMin=<0 -5 -5>, angularTarget=<1 0 0>}
      localJointOffset=<4 0 0>
     }]
    relativeSpaceBone {boneName=spine_03}
   }
   n14 = Anim Dynamics - Bone: dyn_rightHair_A!(ComponentPose=@n13, bAlphaBoolEnabled:bool=True, GravityOverride:Vector="0,0,-980") -> (Pose) {
    alphaInputType=Bool
    angularDampingOverride=1
    angularSpringConstant=10000
    bAngularSpring=true
    bOverrideAngularDamping=true
    bUseGravityOverride=true
    gravityScale=0.35
    simulationSpace=BoneRelative
    alphaBoolBlend {blendInTime=0.1, blendOption=Linear, blendOutTime=0.1}
    boundBone {boneName=dyn_rightHair_A}
    componentAppliedLinearAccClamp=<10000 10000 10000>
    componentLinearAccScale=<1 1 1>
    componentLinearVelScale=<1 1 1>
    gravityOverride {z=-980}
    physicsBodyDefinitions [{
      boundBone {boneName=dyn_rightHair_A}
      boxExtents=<20 20 20>
      constraintSetup {angularLimitsMax=<0 10 10>, angularLimitsMin=<0 -10 -10>, angularTarget=<1 0 0>}
      localJointOffset=<2 0 0>
     }]
    relativeSpaceBone {boneName=spine_03}
   }
   n15 = Anim Dynamics - Bone: dyn_rightHair_B!(ComponentPose=@n14, bAlphaBoolEnabled:bool=True, GravityOverride:Vector="0,0,-980") -> (Pose) {
    alphaInputType=Bool
    angularDampingOverride=1
    angularSpringConstant=10000
    bAngularSpring=true
    bOverrideAngularDamping=true
    bUseGravityOverride=true
    gravityScale=0.35
    simulationSpace=BoneRelative
    alphaBoolBlend {blendInTime=0.1, blendOption=Linear, blendOutTime=0.1}
    boundBone {boneName=dyn_rightHair_B}
    componentAppliedLinearAccClamp=<10000 10000 10000>
    componentLinearAccScale=<1 1 1>
    componentLinearVelScale=<1 1 1>
    gravityOverride {z=-980}
    physicsBodyDefinitions [{
      boundBone {boneName=dyn_rightHair_B}
      boxExtents=<20 20 20>
      constraintSetup {angularLimitsMax=<0 10 10>, angularLimitsMin=<0 -10 -10>, angularTarget=<1 0 0>}
      localJointOffset=<2 0 0>
     }]
    relativeSpaceBone {boneName=spine_03}
   }
   n16 = Anim Dynamics - Bone: dyn_leftHair_A!(ComponentPose=@n15, bAlphaBoolEnabled:bool=True, GravityOverride:Vector="0,0,-980") -> (Pose) {
    alphaInputType=Bool
    angularDampingOverride=1
    angularSpringConstant=10000
    bAngularSpring=true
    bOverrideAngularDamping=true
    bUseGravityOverride=true
    gravityScale=0.35
    simulationSpace=BoneRelative
    alphaBoolBlend {blendInTime=0.1, blendOption=Linear, blendOutTime=0.1}
    boundBone {boneName=dyn_leftHair_A}
    componentAppliedLinearAccClamp=<10000 10000 10000>
    componentLinearAccScale=<1 1 1>
    componentLinearVelScale=<1 1 1>
    gravityOverride {z=-980}
    physicsBodyDefinitions [{
      boundBone {boneName=dyn_leftHair_A}
      boxExtents=<20 20 20>
      constraintSetup {angularLimitsMax=<0 10 10>, angularLimitsMin=<0 -10 -10>, angularTarget=<1 0 0>}
      localJointOffset=<2 0 0>
     }]
    relativeSpaceBone {boneName=spine_03}
   }
   n17 = Anim Dynamics - Bone: dyn_leftHair_B!(ComponentPose=@n16, bAlphaBoolEnabled:bool=True, GravityOverride:Vector="0,0,-980") -> (Pose) {
    alphaInputType=Bool
    angularDampingOverride=1
    angularSpringConstant=10000
    bAngularSpring=true
    bOverrideAngularDamping=true
    bUseGravityOverride=true
    gravityScale=0.35
    simulationSpace=BoneRelative
    alphaBoolBlend {blendInTime=0.1, blendOption=Linear, blendOutTime=0.1}
    boundBone {boneName=dyn_leftHair_B}
    componentAppliedLinearAccClamp=<10000 10000 10000>
    componentLinearAccScale=<1 1 1>
    componentLinearVelScale=<1 1 1>
    gravityOverride {z=-980}
    physicsBodyDefinitions [{
      boundBone {boneName=dyn_leftHair_B}
      boxExtents=<20 20 20>
      constraintSetup {angularLimitsMax=<0 10 10>, angularLimitsMin=<0 -10 -10>, angularTarget=<1 0 0>}
      localJointOffset=<2 0 0>
     }]
    relativeSpaceBone {boneName=spine_03}
   }
   Save cached pose 'CRCache_dynamics'(Pose=Component To Local(ComponentPose=@n17)) {cacheName=CRCache_dynamics}
   n18 = "Transform (Modify) Bone - Bone: base_neckFlap_01_r"!(
    ComponentPose=Local To Component(LocalPose=Use cached pose 'CRCache_constraints'())
    bAlphaBoolEnabled:bool=True
    Translation="0,0,0"
    Rotation="0,0,0"
   ) -> (Pose) {
    alphaInputType=Bool
    rotationMode=BMM_Replace
    rotationSpace=BCS_ParentBoneSpace
    translationMode=BMM_Replace
    translationSpace=BCS_ParentBoneSpace
    alphaBoolBlend {blendOption=Linear}
    boneToModify {boneName=base_neckFlap_01_r}
   }
   n19 = "Transform (Modify) Bone - Bone: dyn_neckFlap_01_r"!(ComponentPose=@n18, bAlphaBoolEnabled:bool=True, Translation="0,0,0", Rotation="0,0,0") -> (Pose) {
    alphaInputType=Bool
    rotationMode=BMM_Replace
    rotationSpace=BCS_ParentBoneSpace
    translationMode=BMM_Replace
    translationSpace=BCS_ParentBoneSpace
    alphaBoolBlend {blendOption=Linear}
    boneToModify {boneName=dyn_neckFlap_01_r}
   }
   n20 = "Transform (Modify) Bone - Bone: base_neckFlap_02_r"!(ComponentPose=@n19, bAlphaBoolEnabled:bool=True, Translation="0,0,0", Rotation="0,0,0") -> (Pose) {
    alphaInputType=Bool
    rotationMode=BMM_Replace
    rotationSpace=BCS_ParentBoneSpace
    translationMode=BMM_Replace
    translationSpace=BCS_ParentBoneSpace
    alphaBoolBlend {blendOption=Linear}
    boneToModify {boneName=base_neckFlap_02_r}
   }
   n21 = "Transform (Modify) Bone - Bone: dyn_neckFlap_02_r"!(ComponentPose=@n20, bAlphaBoolEnabled:bool=True, Translation="0,0,0", Rotation="0,0,0") -> (Pose) {
    alphaInputType=Bool
    rotationMode=BMM_Replace
    rotationSpace=BCS_ParentBoneSpace
    translationMode=BMM_Replace
    translationSpace=BCS_ParentBoneSpace
    alphaBoolBlend {blendOption=Linear}
    boneToModify {boneName=dyn_neckFlap_02_r}
   }
   n22 = "Transform (Modify) Bone - Bone: base_neckFlap_b"!(ComponentPose=@n21, bAlphaBoolEnabled:bool=True, Translation="0,0,0", Rotation="0,0,0") -> (Pose) {
    alphaInputType=Bool
    rotationMode=BMM_Replace
    rotationSpace=BCS_ParentBoneSpace
    translationMode=BMM_Replace
    translationSpace=BCS_ParentBoneSpace
    alphaBoolBlend {blendOption=Linear}
    boneToModify {boneName=base_neckFlap_b}
   }
   n23 = "Transform (Modify) Bone - Bone: dyn_neckFlap_b"!(ComponentPose=@n22, bAlphaBoolEnabled:bool=True, Translation="0,0,0", Rotation="0,0,0") -> (Pose) {
    alphaInputType=Bool
    rotationMode=BMM_Replace
    rotationSpace=BCS_ParentBoneSpace
    translationMode=BMM_Replace
    translationSpace=BCS_ParentBoneSpace
    alphaBoolBlend {blendOption=Linear}
    boneToModify {boneName=dyn_neckFlap_b}
   }
   n24 = "Transform (Modify) Bone - Bone: base_jacktFlap_r"!(ComponentPose=@n23, bAlphaBoolEnabled:bool=True, Translation="0,0,0", Rotation="0,0,0") -> (Pose) {
    alphaInputType=Bool
    rotationMode=BMM_Replace
    rotationSpace=BCS_ParentBoneSpace
    translationMode=BMM_Replace
    translationSpace=BCS_ParentBoneSpace
    alphaBoolBlend {blendOption=Linear}
    boneToModify {boneName=base_jacktFlap_r}
   }
   n25 = "Transform (Modify) Bone - Bone: dyn_jacktFlap_r"!(ComponentPose=@n24, bAlphaBoolEnabled:bool=True, Translation="0,0,0", Rotation="0,0,0") -> (Pose) {
    alphaInputType=Bool
    rotationMode=BMM_Replace
    rotationSpace=BCS_ParentBoneSpace
    translationMode=BMM_Replace
    translationSpace=BCS_ParentBoneSpace
    alphaBoolBlend {blendOption=Linear}
    boneToModify {boneName=dyn_jacktFlap_r}
   }
   Save cached pose 'CRCache_JacketFlaps'(Pose=Component To Local(ComponentPose=@n25)) {cacheName=CRCache_JacketFlaps}
   n26 = "Transform (Modify) Bone - Bone: dyn_ponytail_01"!(
    ComponentPose=Local To Component(LocalPose=Use cached pose 'CRCache_dynamics'())
    bAlphaBoolEnabled:bool=True
    Translation="0,0,0"
    Rotation="10,0,20"
   ) -> (Pose) {
    alphaInputType=Bool
    rotationMode=BMM_Additive
    translationMode=BMM_Additive
    translationSpace=BCS_BoneSpace
    alphaBoolBlend {blendOption=Linear}
    boneToModify {boneName=dyn_ponytail_01}
    rotation {pitch=10, roll=20}
   }
   n27 = "Transform (Modify) Bone - Bone: dyn_ponytail_02"!(ComponentPose=@n26, bAlphaBoolEnabled:bool=True, Translation="0,5,-2", Rotation="-10,15,-15") -> (Pose) {
    alphaInputType=Bool
    rotationMode=BMM_Additive
    translationMode=BMM_Additive
    translationSpace=BCS_BoneSpace
    alphaBoolBlend {blendInTime=0.2, blendOption=Linear, blendOutTime=0.2}
    boneToModify {boneName=dyn_ponytail_02}
    rotation=<rot -10 15 -15>
    translation {y=5, z=-2}
   }
   n28 = "Transform (Modify) Bone - Bone: dyn_ponytail_03"!(ComponentPose=@n27, bAlphaBoolEnabled:bool=True, Translation="0,0,0", Rotation="0,0,-15") -> (Pose) {
    alphaInputType=Bool
    rotationMode=BMM_Additive
    translationMode=BMM_Additive
    translationSpace=BCS_BoneSpace
    alphaBoolBlend {blendOption=Linear}
    boneToModify {boneName=dyn_ponytail_03}
    rotation {roll=-15}
   }
   n29 = "Transform (Modify) Bone - Bone: dyn_ponytail_02"!(ComponentPose=@n28, Alpha:float=1, Translation="-2,0,0", Rotation="-5,0,0") -> (Pose) {
    rotationMode=BMM_Additive
    translationMode=BMM_Additive
    alphaBoolBlend {blendOption=Linear}
    alphaScaleBiasClamp {bInterpResult=true, interpSpeedDecreasing=5, interpSpeedIncreasing=30}
    boneToModify {boneName=dyn_ponytail_02}
    rotation {pitch=-5}
    translation {x=-2}
   }
   n30 = "Transform (Modify) Bone - Bone: dyn_ponytail_03"!(ComponentPose=@n29, Alpha:float=1, Translation="-2,0,0", Rotation="-10,0,0") -> (Pose) {
    rotationMode=BMM_Additive
    translationMode=BMM_Additive
    alphaBoolBlend {blendOption=Linear}
    alphaScaleBiasClamp {bInterpResult=true, interpSpeedDecreasing=5, interpSpeedIncreasing=30}
    boneToModify {boneName=dyn_ponytail_03}
    rotation {pitch=-10}
    translation {x=-2}
   }
   n31 = "Transform (Modify) Bone - Bone: dyn_ponytail_04"!(ComponentPose=@n30, Alpha=1, Translation="-2,0,0", Rotation="-15,0,0") -> (Pose) {
    rotationMode=BMM_Additive
    translationMode=BMM_Additive
    alphaBoolBlend {blendOption=Linear}
    alphaScaleBiasClamp {bInterpResult=true, interpSpeedDecreasing=5, interpSpeedIncreasing=30}
    boneToModify {boneName=dyn_ponytail_04}
    rotation {pitch=-15}
    translation {x=-2}
   }
   Save cached pose 'CRCache_offset'(Pose=Component To Local(ComponentPose=@n31)) {cacheName=CRCache_offset}
  ]}
 EventGraph {nodes [
   n1 = Event BlueprintUpdateAnimation()
   > n2 = Sequence!(execute) -> (then_0, then_1, then_2:Exec)
   > n3 = Gravity Override State Function DEPRECATED(CharacterPartAnimInstance=Self-Reference(), GameTimeInSeconds=GetGameTimeInSeconds(), GravityOverrideParams=Get Gravity Override Params())
   > Set GravityOverride(GravityOverride=@n3)
   n4 = Get bIsTargeting()
   > Set isTargeting(isTargeting:double=FInterpTo(Current=Get isTargeting(), Target="To Float (Boolean)"(InBool=@n4), DeltaTime=@n1.DeltaTimeX, InterpSpeed=15)) -> (Output_Get:double)
   n5 = GetAnimInstance(self=GetOwnerSkeletalMeshComponent())
   Is Valid(exec=@n2.then_2, InputObject=@n5)
   > Set Is Custom Emote Playing(Is Custom Emote Playing:bool=Montage_IsPlaying(self=@n5, Montage=/BRCosmetics/Animation/Game/MainPlayer/Emotes/ZombieWalk/Emote_ZombieWalk_CMF_M)) -> (Output_Get:bool)
   Set CrouchFixAlpha(execute=@n2.then_1, CrouchFixAlpha:double=GetCurveValue(CurveName=Remap_CrouchFixCrv)) -> (Output_Get:double)
   n6 = Get Body_Mesh()
   > Is Valid(InputObject=@n6)
   n7 = GetSocketBoneName!(self:Skinned Mesh Component Object Reference=@n6, InSocketName:Name=upperarm_r) -> (ReturnValue:Name=None)
   > n8 = Set isBodyAttached(isBodyAttached:bool="Not Equal Exactly (String)"(A="To String (Name)"(InName=@n7), B=None)) -> (Output_Get:bool)
   > Branch(Condition=@n8)
   > GetSocketTransformFromBody()
   Event OnParachuteOpened()
   > Switch on EGliderType(Selection=Get PlayerGliderType() -> (PlayerGliderType=HangGlider))
   > Set HangGliding(HangGliding:bool=true) -> (Output_Get:bool)
   > Set HairDynAlpha(HairDynAlpha:double=0) -> (Output_Get:double)
   Event OnParachuteClosed()
   > Set HangGliding(HangGliding=false)
   > Set HairDynAlpha(HairDynAlpha=1)
   Event OnBeginCrouching()
   > Set fCrouching(fCrouching:bool=1) -> (Output_Get:bool)
   Event OnWeaponChanged()
   > Set isGunOut(isGunOut:double=1) -> (Output_Get:double)
   > n9 = Switch on EFortWeaponCoreAnimation!(execute, Selection:EFortWeaponCoreAnimation Enum=Get WeaponCoreAnim!() -> (WeaponCoreAnim)) -> (
    Melee:Exec
    Pistol:Exec
    Shotgun:Exec
    PaperBlueprint:Exec
    Rifle:Exec
    MeleeOneHand:Exec
    MachinePistol:Exec
    RocketLauncher:Exec
    GrenadeLauncher:Exec
    GoingCommando:Exec
    AssaultRifle:Exec
    TacticalShotgun:Exec
    SniperRifle:Exec
    TrapPlacement:Exec
    ShoulderLauncher:Exec
    AbilityDecoTool:Exec
    Crossbow:Exec
    C4:Exec
    RemoteControl:Exec
    DualWield:Exec
    AR_BullPup:Exec
    AR_ForwardGrip:Exec
    MedPackPaddles:Exec
    SMG_P90:Exec
    AR_DrumGun:Exec
    Consumable_Small:Exec
    Consumable_Large:Exec
    Balloon:Exec
    MountedTurret:Exec
    CreativeTool:Exec
    ExplosiveBow:Exec
    AshtonIndigo:Exec
    AshtonChicago:Exec
    MeleeDualWield:Exec
    Unarmed:Exec
   )
   Set isGunOut(execute=[@n9.Melee, @n9.MeleeDualWield, @n9.MeleeOneHand, @n9.Unarmed], isGunOut=0)
   n10 = Event OnCharacterCustomizationLoaded()
   > Set Body_Mesh(Body_Mesh=Get Array Item with EFort Custom Part Type(Array=@n10.CustomMeshParts))
   n11 = CheckForBackbling()
   > Set isBackpackEquipped(isBackpackEquipped:bool=AND Boolean(A=@n11, B=NOT Boolean(A=@n11.bIsBackblingACape))) -> (Output_Get:bool)
   Event OnEndCrouching()
   > Set fCrouching(fCrouching=false)
  ]}
 GetSocketTransformFromBody {nodes [
   GetSocketTransformFromBody() {is=title K2Node_FunctionEntry}
   n1 = BreakTransform(InTransform=GetSocketTransform(self=Get Body_Mesh(), InSocketName=base_jacktFlap_r, TransformSpace=RTS_ParentBoneSpace))
   > Set base_jacktFlap_r_Translation(base_jacktFlap_r_Translation:Vector=@n1) -> (Output_Get:Vector)
   > Set base_jacktFlap_r_Rotation(base_jacktFlap_r_Rotation:Rotator=@n1.Rotation) -> (Output_Get:Rotator)
   n2 = BreakTransform(InTransform=GetSocketTransform(self=Get Body_Mesh(), InSocketName=dyn_jacktFlap_r, TransformSpace=RTS_ParentBoneSpace))
   > Set dyn_jacktFlap_r_Translation_0(dyn_jacktFlap_r_Translation_0:Vector=@n2) -> (Output_Get:Vector)
   > Set dyn_jacktFlap_r_Rotation_0(dyn_jacktFlap_r_Rotation_0:Rotator=@n2.Rotation) -> (Output_Get:Rotator)
   n3 = BreakTransform(InTransform=GetSocketTransform(self=Get Body_Mesh(), InSocketName=base_neckFlap_01_r, TransformSpace=RTS_ParentBoneSpace))
   > Set base_neckFlap_01_r_Translation(base_neckFlap_01_r_Translation:Vector=@n3) -> (Output_Get:Vector)
   > Set base_neckFlap_01_r_Rotation(base_neckFlap_01_r_Rotation:Rotator=@n3.Rotation) -> (Output_Get:Rotator)
   n4 = BreakTransform(InTransform=GetSocketTransform(self=Get Body_Mesh(), InSocketName=dyn_neckFlap_01_r, TransformSpace=RTS_ParentBoneSpace))
   > Set dyn_neckFlap_01_r_Translation(dyn_neckFlap_01_r_Translation:Vector=@n4) -> (Output_Get:Vector)
   > Set dyn_neckFlap_01_r_Rotation(dyn_neckFlap_01_r_Rotation:Rotator=@n4.Rotation) -> (Output_Get:Rotator)
   n5 = BreakTransform(InTransform=GetSocketTransform(self=Get Body_Mesh(), InSocketName=base_neckFlap_02_r, TransformSpace=RTS_ParentBoneSpace))
   > Set base_neckFlap_02_r_Translation(base_neckFlap_02_r_Translation:Vector=@n5) -> (Output_Get:Vector)
   > Set base_neckFlap_02_r_Rotation(base_neckFlap_02_r_Rotation:Rotator=@n5.Rotation) -> (Output_Get:Rotator)
   n6 = BreakTransform(InTransform=GetSocketTransform(self=Get Body_Mesh(), InSocketName=dyn_neckFlap_02_r, TransformSpace=RTS_ParentBoneSpace))
   > Set dyn_neckFlap_02_r_Translation(dyn_neckFlap_02_r_Translation:Vector=@n6) -> (Output_Get:Vector)
   > Set dyn_neckFlap_02_r_Rotation(dyn_neckFlap_02_r_Rotation:Rotator=@n6.Rotation) -> (Output_Get:Rotator)
   n7 = BreakTransform(InTransform=GetSocketTransform(self=Get Body_Mesh(), InSocketName=base_neckFlap_b, TransformSpace=RTS_ParentBoneSpace))
   > Set base_neckFlap_b_Translation(base_neckFlap_b_Translation:Vector=@n7) -> (Output_Get:Vector)
   > Set base_neckFlap_b_Rotation(base_neckFlap_b_Rotation:Rotator=@n7.Rotation) -> (Output_Get:Rotator)
   n8 = BreakTransform(InTransform=GetSocketTransform(self=Get Body_Mesh(), InSocketName=dyn_neckFlap_b, TransformSpace=RTS_ParentBoneSpace))
   > Set dyn_neckFlap_b_Tranform(dyn_neckFlap_b_Tranform:Vector=@n8) -> (Output_Get:Vector)
   > Set dyn_neckFlap_b_Rotation(dyn_neckFlap_b_Rotation:Rotator=@n8.Rotation) -> (Output_Get:Rotator)
  ]}
]
variables [
 base_jacktFlap_r_Rotation {type=Rotator}
 base_jacktFlap_r_Translation {type=Vector}
 base_neckFlap_01_l_Rotation {type=Rotator}
 base_neckFlap_01_l_Translation {type=Vector}
 base_neckFlap_01_r_Rotation {type=Rotator}
 base_neckFlap_01_r_Translation {type=Vector}
 base_neckFlap_02_l_Rotation {type=Rotator}
 base_neckFlap_02_l_Translation {type=Vector}
 base_neckFlap_02_r_Rotation {type=Rotator}
 base_neckFlap_02_r_Translation {type=Vector}
 base_neckFlap_b_Rotation {type=Rotator}
 base_neckFlap_b_Translation {type=Vector}
 Body_Mesh {type=SkeletalMeshComponent}
 CrouchFixAlpha {type=number}
 dyn_jacktFlap_r_Rotation_0 {type=Rotator}
 dyn_jacktFlap_r_Translation_0 {type=Vector}
 dyn_neckFlap_01_l_Rotation {type=Rotator}
 dyn_neckFlap_01_l_Translation {type=Vector}
 dyn_neckFlap_01_r_Rotation {type=Rotator}
 dyn_neckFlap_01_r_Translation {type=Vector}
 dyn_neckFlap_02_l_Rotation {type=Rotator}
 dyn_neckFlap_02_l_Translation {type=Vector}
 dyn_neckFlap_02_r_Rotation {type=Rotator}
 dyn_neckFlap_02_r_Translation {type=Vector}
 dyn_neckFlap_b_Rotation {type=Rotator}
 dyn_neckFlap_b_Tranform {type=Vector}
 fCrouching {type=boolean}
 Gravity Override Params {type=GravityOverrideParamsStruct}
 GravityOverride {type=Vector}
 HairDynAlpha {type=number, default=1}
 HangGliding {type=boolean}
 Is Custom Emote Playing {type=boolean}
 isBackpackEquipped {type=boolean}
 isBodyAttached {type=boolean}
 isGunOut {type=number}
 isTargeting {type=number}
]
